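# HUD toast widget showing playlist travel status, with destination preview, state text, countdown and squad travel opt-out.
# Travel status selects the text and drives two looping timers: animated launching dots and a prep-time progress fill on the backplate material.
# Prep time optionally lerps toward updated values to smooth the countdown.
# View model images feed the preview's dynamic material, and a downloading state swaps out the travel info.
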WidgetBlueprint /FortPlaylistTravelUI/HUD/WBP_FortPlaylistTravel_HUDToast
animations [
 Anim_Intro {
  binding_count=3
  length_seconds=0.6
  object=&self:Anim_Intro
  sequence {bindings [
    Border_Backing {
     id=#1
     possessed_class=Border
     tracks [{
       type=MovieSceneWidgetMaterialTrack
       sections [{
         type=MovieSceneParameterSection
         end_frame=12
         start_frame=0
         channels [Animation Control {type=MovieSceneScriptingFloatChannel, keys [[0,0],[12,1]]}]
        }]
      }]
    }
    DestinationPreview {
     id=#2
     possessed_class=FortLazyImage
     tracks [{
       type=MovieScene2DTransformTrack
       sections [{
         type=MovieScene2DTransformSection
         end_frame=12
         start_frame=0
         channels [
          Scale.X {type=MovieSceneScriptingFloatChannel, keys [[7,0],[10,1.1],[12,1]]}
          Scale.Y {type=MovieSceneScriptingFloatChannel, keys [[7,0],[10,1.1],[12,1]]}
         ]
        }]
      }]
    }
    WBP_FortPlaylistTravel_HUDToast_C_55 {
     id=#3
     possessed_class=WBP_FortPlaylistTravel_HUDToast_C
     tracks [
      {
       type=MovieSceneFloatTrack
       display_name=Render Opacity
       sections [{type=MovieSceneFloatSection, end_frame=12, start_frame=0, channels [None {type=MovieSceneScriptingFloatChannel, keys [[0,0],[4.0003,1]]}]}]
      }
      {
       type=MovieScene2DTransformTrack
       sections [{
         type=MovieScene2DTransformSection
         end_frame=11
         start_frame=0
         channels [Translation.X {type=MovieSceneScriptingFloatChannel, keys [[0,100],[4.0003,0]]}]
        }]
      }
     ]
    }
   ]}
 }
 Anim_Outro {
  binding_count=1
  length_seconds=0.2
  object=&self:Anim_Outro
  sequence {bindings [WBP_FortPlaylistTravel_HUDToast_C_56 {
     id=#4
     possessed_class=WBP_FortPlaylistTravel_HUDToast_C
     tracks [
      {
       type=MovieScene2DTransformTrack
       sections [{
         type=MovieScene2DTransformSection
         end_frame=4
         start_frame=0
         channels [Translation.X {type=MovieSceneScriptingFloatChannel, keys [[0,0],[4,40]]}]
        }]
      }
      {
       type=MovieSceneFloatTrack
       display_name=Render Opacity
       sections [{type=MovieSceneFloatSection, end_frame=4, start_frame=0, channels [None {type=MovieSceneScriptingFloatChannel, keys [[0,1],[4,0]]}]}]
      }
     ]
    }]}
 }
]
graphs [
 BGImageLoadingState {nodes [
   n1 = BGImageLoadingState!() -> (then, LoadingState:EActivityImageLoadingState Enum=Unset) {is=title K2Node_FunctionEntry}
   > n2 = Switch on EActivityImageLoadingState!(execute, Selection:EActivityImageLoadingState Enum=@n1) -> (Unset:Exec, Loading:Exec, Loaded:Exec, Error:Exec) {is=title K2Node_SwitchEnum}
   n3 = Get CurrentGameTravelDestinationViewModel!(self:Fort Playlist Travel View Model Object Reference=Get BP_FortPlaylistTravelViewModel() -> (BP_FortPlaylistTravelViewModel:BP Fort Playlist Travel View Model Object Reference)) -> (CurrentGameTravelDestinationViewModel:Fort Playlist Travel Entry View Model Object Reference)
   n4 = Branch(
    execute=@n2.Loaded
    Condition=IsValidSoftObjectReference(SoftObjectReference=Get SmallPreviewImage!(self:Fort Playlist Travel Entry View Model Object Reference=@n3) -> (SmallPreviewImage:Soft Object Reference))
   )
   n5 = GetDynamicMaterial(execute=@n4.else, self=Get DestinationPreview() -> (DestinationPreview:Fort Lazy Image Object Reference))
   n6 = Get ActivityViewModel!(self:Fort Playlist Travel Entry View Model Object Reference=Get CurrentGameTravelDestinationViewModel(self=Get BP_FortPlaylistTravelViewModel())) -> (ActivityViewModel:Discover Activity View Model Object Reference)
   > SetTextureParameterValue(
    self=@n5
    ParameterName=Image
    Value=GetImageTexture!(self:Fort Discover Link Entry Tile Item VM Object Reference=@n6) -> (ReturnValue:Texture Object Reference)
   )
   n7 = BreakVector2D(InVec=GetTextureSize!(self:Fort Discover Link Entry Tile Item VM Object Reference=@n6) -> (ReturnValue:Vector 2D Structure))
   > SetScalarParameterValue(self=@n5, ParameterName=X_TexSize, Value=@n7)
   > SetScalarParameterValue(self=@n5, ParameterName=Y_TexSize, Value=@n7.Y)
   > SetVisibility(self=Get DestinationPreview(), InVisibility=Visible)
  ]}
 EventGraph {nodes [
   StartTimer_Launching() {is=title K2Node_CustomEvent}
   > n1 = Clear and Invalidate Timer by Handle(Handle=Get Timer_Preptime_Progress() -> (Timer_Preptime_Progress:Timer Handle Structure))
   n2 = OnTimerEvent_Launching() {is=title K2Node_CustomEvent}
   n3 = Set Timer by Event(execute=@n1.then, Delegate=@n2, Time=Get LaunchDotUpdateTime() -> (LaunchDotUpdateTime:float), bLooping=true, bMaxOncePerFrame=true)
   > Set Timer_Launching(Timer_Launching:Timer Handle Structure=@n3) -> (Output_Get:Timer Handle Structure)
   n4 = Get CurrTravelStatus() -> (CurrTravelStatus:EFortPlaylistTravelStatus Enum=None)
   > FortShippingLog!(
    execute
    WorldContextObject:Object Reference=Self-Reference()
    InString:String=Append(A="[PlaylistTravel] [HUD] StartTimer_Launching: ", B=Enum to String!(Enumerator:EFortPlaylistTravelStatus Enum=@n4) -> (ReturnValue:String))
    bLogAsWarning:bool=false
   ) -> (then)
   UpdateLaunchingText() {is=title K2Node_CustomEvent}
   > Branch(Condition="Equal (Enum)"(A:EFortPlaylistTravelStatus Enum=Get CurrTravelStatus() -> (CurrTravelStatus=None), B:EFortPlaylistTravelStatus Enum=Launching))
   n5 = Select!(Option 0:String=., Option 1:String=.., Option 2:String=..., Index:int=Get LaunchingTextDotCount() -> (LaunchingTextDotCount:int=0)) -> (ReturnValue:String)
   > SetText(self=Get Text_TravelState_Dot() -> (Text_TravelState_Dot:Common Rich Text Block Object Reference), InText="To Text (String)"(InString=@n5))
   n6 = Get LaunchingTextDotCount() -> (LaunchingTextDotCount=0)
   > Set LaunchingTextDotCount(LaunchingTextDotCount:int="% (Integer)"(A=int + int(A=@n6), B=3) -> (ReturnValue=0)) -> (Output_Get:int=0)
   n7 = OnProgressTimerEvent() {is=title K2Node_CustomEvent}
   > n8 = Sequence!(execute) -> (then_0, then_1, then_2:Exec)
   n9 = Get CurrPrepTimeLeftSec() -> (CurrPrepTimeLeftSec:double)
   n10 = Get UpdatePrepTimeLeftSec() -> (UpdatePrepTimeLeftSec:double)
   > n11 = Branch(Condition=AND Boolean(A="float > float"(A=@n9, B=@n10), B=Get UsePrepTimeInterpolation() -> (UsePrepTimeInterpolation:bool)))
   n12 = float * float(A=Get PrepTimeTickSec() -> (PrepTimeTickSec:double), B=Get PrepTimeFixInterpolateSpeed() -> (PrepTimeFixInterpolateSpeed:double)) {is=title K2Node_PromotableOperator}
   n13 = Get CurrPrepTimeLeftSec()
   n14 = Get UpdatePrepTimeLeftSec()
   > Set CurrPrepTimeLeftSec(CurrPrepTimeLeftSec:double=float - float(A=Lerp(A=@n13, B=@n14, Alpha=@n12), B=Get PrepTimeTickSec())) -> (Output_Get:double)
   Set CurrPrepTimeLeftSec(execute=@n11.else, CurrPrepTimeLeftSec=float - float(A=Get CurrPrepTimeLeftSec(), B=Get PrepTimeTickSec()))
   n15 = float / float(A=Get CurrPrepTimeLeftSec(), B=Get TotalPrepTimeLeftSec() -> (TotalPrepTimeLeftSec:double))
   n16 = Set ProgressTime01(execute=@n8.then_1, ProgressTime01:double=float - float(A=1, B="Clamp (Float)"(Value=@n15, Min=0, Max=1))) -> (Output_Get:double)
   > UpdateProgressBar!(execute, self, Progress:double=@n16) -> (then)
   n17 = Get ProgressTime01() -> (ProgressTime01:double)
   n18 = Get CurrTravelStatus() -> (CurrTravelStatus=None)
   FortShippingLog(
    execute=@n8.then_2
    WorldContextObject=Self-Reference()
    InString=Append!(A="[PlaylistTravel] [HUD] OnProgressTimer: ", B="To String (Float)"(InDouble=@n17), C:String=", ", D:String=Enum to String(Enumerator=@n18)) -> (ReturnValue)
    bLogAsWarning=false
   )
   StartProgressTimer() {is=title K2Node_CustomEvent}
   > Clear and Invalidate Timer by Handle(Handle=Get Timer_Preptime_Progress())
   > n19 = Set Timer by Event(Delegate=@n7, Time=Get PrepTimeTickSec(), bLooping=true, bMaxOncePerFrame=true)
   > Set Timer_Preptime_Progress(Timer_Preptime_Progress:Timer Handle Structure=@n19) -> (Output_Get:Timer Handle Structure)
   n20 = Get CurrPrepTimeLeftSec()
   n21 = Get CurrTravelStatus() -> (CurrTravelStatus=None)
   > FortShippingLog(
    WorldContextObject=Self-Reference()
    InString=Append(A="[PlaylistTravel] [HUD] StartProgressTimer: ", B="To String (Float)"(InDouble=@n20), C=", ", D=Enum to String(Enumerator=@n21))
    bLogAsWarning=false
   )
   Event Construct()
   > n22 = GetDynamicMaterial(self=Get DestinationPreview())
   > SetScalarParameterValue(self=@n22, ParameterName=X_TexSize, Value=64)
   > SetScalarParameterValue(self=@n22, ParameterName=Y_TexSize, Value=64)
   > SetVisibility(self=Get SB_ClippingEnforcement_Download() -> (SB_ClippingEnforcement_Download:Size Box Object Reference), InVisibility=Collapsed)
   n23 = GetOwningPlayer()
   Branch(Condition=IsValid(Object=@n23))
   > Branch(Condition=IsUsingTouch(PlayerController=@n23))
   n24 = Get BP_FortPlaylistTravelViewModel()
   n25 = GetComponentByClass!(
    self:Actor Object Reference=GetOwningPlayer()
    ComponentClass:Actor Component Class Reference=/Script/FortPlaylistTravelRuntime.FortControllerComponent_PlaylistTravel
   ) -> (ReturnValue:Fort Controller Component Playlist Travel Object Reference)
   > Branch(Condition=AND Boolean(A=IsValid(Object=@n24), B=IsValid(Object=@n25)))
   > Branch(Condition=Get bIsSquadTravel!(self:Fort Playlist Travel View Model Object Reference=@n24) -> (bIsSquadTravel:bool))
   > OnSquadOptOutInputTriggered(self:Fort Controller Component Playlist Travel Object Reference=@n25)
   UpdateLaunchingText(execute=@n2.then) {is=title K2Node_CallFunction}
   Append!(A="To String (Text)"(InText=Get LaunchingText() -> (LaunchingText:Text)), B=@n5) -> (ReturnValue)
   "On Clicked (UIFrameworkButtonWidget_Toast)"() {is=title K2Node_ComponentBoundEvent}
  ]}
 HandleDownloading {nodes [
   n1 = HandleDownloading!() -> (then, IsDownloading:bool=~) {is=title K2Node_FunctionEntry}
   > n2 = Branch(Condition=@n1)
   > SetVisibility(self=Get SB_ClippingEnforcement_Download(), InVisibility=Visible)
   > SetVisibility(self=Get SB_ClippingEnforcement_TravelInfo() -> (SB_ClippingEnforcement_TravelInfo:Size Box Object Reference), InVisibility=Collapsed)
   SetVisibility(execute=@n2.else, self=Get SB_ClippingEnforcement_Download(), InVisibility=Collapsed)
   > SetVisibility(self=Get SB_ClippingEnforcement_TravelInfo(), InVisibility=Visible)
  ]}
 HandleGameTravelStatus {nodes [
   n1 = HandleGameTravelStatus!() -> (then, Status:EFortPlaylistTravelStatus Enum=None) {is=title K2Node_FunctionEntry}
   > n2 = Sequence!(execute) -> (then_0, then_1)
   > Set CurrTravelStatus(CurrTravelStatus:EFortPlaylistTravelStatus Enum=@n1) -> (Output_Get:EFortPlaylistTravelStatus Enum=None)
   > SetText(
    self=Get Text_TravelState() -> (Text_TravelState:Common Rich Text Block Object Reference)
    InText=Select!(
     None:"Text (by ref)"
     PrepTime:"Text (by ref)"=Get PreparingText() -> (PreparingText:Text)
     Launching:"Text (by ref)"=Get LaunchingText()
     Matchmaking:"Text (by ref)"=Get MatchmakingText() -> (MatchmakingText:Text)
     MatchmakingSuccess:"Text (by ref)"
     MatchmakingFailed:"Text (by ref)"
     Index:EFortPlaylistTravelStatus Enum=@n1
    ) -> (ReturnValue:"Text (by ref)")
   )
   > SetVisibility(
    self=Get Text_Timer() -> (Text_Timer:Common Rich Text Block Object Reference)
    InVisibility=Select(
     None:ESlateVisibility Enum=Collapsed
     PrepTime:ESlateVisibility Enum=Visible
     Launching:ESlateVisibility Enum=Collapsed
     Matchmaking:ESlateVisibility Enum=Collapsed
     MatchmakingSuccess:ESlateVisibility Enum=Collapsed
     MatchmakingFailed:ESlateVisibility Enum=Collapsed
     Index=Get CurrTravelStatus() -> (CurrTravelStatus=None)
    ) -> (ReturnValue:ESlateVisibility Enum)
   )
   n3 = Get CurrTravelStatus() -> (CurrTravelStatus=None)
   > SetVisibility(
    self=Get Text_TravelState_Dot()
    InVisibility=Select!(Option 0:ESlateVisibility Enum=Collapsed, Option 1:ESlateVisibility Enum=Visible, Index:bool="Equal (Enum)"(A=@n3, B=Launching)) -> (ReturnValue)
   )
   > n4 = Sequence()
   > n5 = Branch(Condition="Equal (Enum)"(A=Get CurrTravelStatus() -> (CurrTravelStatus=None), B=PrepTime))
   > UpdateProgressBar(Progress=1)
   > Clear and Invalidate Timer by Handle(Handle=Get Timer_Preptime_Progress())
   n6 = GetRemainingPrepTime!(self:Fort Playlist Travel View Model Object Reference=Get BP_FortPlaylistTravelViewModel()) -> (ReturnValue:Timespan Structure)
   n7 = Get MaxPrepTimeSec() -> (MaxPrepTimeSec:int=0)
   n8 = Set CurrPrepTimeLeftSec(
    execute=@n5.else
    CurrPrepTimeLeftSec="Clamp (Float)"(
     Value=Get Total Seconds!(A:Timespan Structure=@n6) -> (ReturnValue:double) {is=title K2Node_CallFunction}
     Min=1
     Max="To Float (Integer)"(InInt=@n7)
    )
   )
   > n9 = Set TotalPrepTimeLeftSec(TotalPrepTimeLeftSec:double=@n8) -> (Output_Get:double)
   > Set UpdatePrepTimeLeftSec(UpdatePrepTimeLeftSec:double=@n9) -> (Output_Get:double)
   > StartProgressTimer() {is=title K2Node_CallFunction}
   n10 = Branch(execute=@n4.then_1, Condition="Equal (Enum)"(A=Get CurrTravelStatus() -> (CurrTravelStatus=None), B=Launching))
   n11 = Get Timer_Launching() -> (Timer_Launching:Timer Handle Structure)
   > Branch(Condition=Is Valid Timer Handle!(Handle:Timer Handle Structure=@n11) -> (ReturnValue:bool))
   > Clear and Invalidate Timer by Handle(Handle=@n11)
   StartTimer_Launching(execute=@n10.else) {is=title K2Node_CallFunction}
   n12 = Get CurrTravelStatus() -> (CurrTravelStatus=None)
   FortShippingLog(
    execute=@n2.then_1
    WorldContextObject=Self-Reference()
    InString=Append(A="[PlaylistTravel] [HUD] HandlePlaylistStatus: ", B=Enum to String(Enumerator=@n12))
    bLogAsWarning=false
   )
  ]}
 HandleImageFromViewModel {nodes [
   n1 = HandleImageFromViewModel!() -> (then, PreviewImage:Soft Object Reference) {is=title K2Node_FunctionEntry}
   > Branch(Condition=IsValidSoftObjectReference(SoftObjectReference=@n1))
   > SetVisibility(self=Get DestinationPreview(), InVisibility=Visible)
  ]}
 OnActivityViewModelSet {nodes [
   n1 = OnActivityViewModelSet!() -> (then, ActivityVM:Discover Activity View Model Object Reference) {is=title K2Node_FunctionEntry}
   > Is Valid(InputObject=@n1)
   > FetchImages(self:Fort Discover Link Entry Tile Item VM Object Reference=Get ActivityVM() -> (ActivityVM:Discover Activity View Model Object Reference))
   n2 = Get CurrentGameTravelDestinationViewModel(self=Get BP_FortPlaylistTravelViewModel())
   > n3 = Branch(Condition=IsValidSoftObjectReference(SoftObjectReference=Get SmallPreviewImage(self=@n2)))
   SetVisibility(execute=@n3.else, self=Get DestinationPreview(), InVisibility=Collapsed)
  ]}
 Sequencer Events {links [], nodes []}
 SetDestinationText {nodes [
   SetText(self=Get Text_TravelState())
   SetDestinationText() {is=title K2Node_FunctionEntry}
  ]}
 UpdatePrepTime {nodes [
   n1 = UpdatePrepTime!() -> (then, Timespan:Timespan Structure) {is=title K2Node_FunctionEntry}
   n2 = BreakTimespan2!(InTimespan:Timespan Structure=@n1) -> (Days:int=0, Hours:int=0, Minutes:int=0, Seconds:int=0, FractionNano:int=0)
   n3 = MakeTimespan2!(
    Days:int=@n2
    Hours:int=@n2.Hours
    Minutes:int=@n2.Minutes
    Seconds:int="Clamp (Integer)"(Value=@n2.Seconds, Min=0, Max=Get MaxPrepTimeSec() -> (MaxPrepTimeSec=0)) -> (ReturnValue=0)
    FractionNano:int=@n2.FractionNano
   ) -> (ReturnValue:Timespan Structure)
   > SetText(self=Get Text_Timer(), InText=As Timespan!(InTimespan:"Timespan Structure (by ref)"=@n3) -> (ReturnValue:Text))
   n4 = Get Total Seconds(A=@n3)
   n5 = Get CurrPrepTimeLeftSec()
   > Set UpdatePrepTimeLeftSec(UpdatePrepTimeLeftSec="Max (Float)"(A="Min (Float)"(A:double=@n4, B:double=@n5) -> (ReturnValue:double) {is=title K2Node_CommutativeAssociativeBinaryOperator}, B=0))
   > n6 = Sequence()
   > n7 = Branch(Condition=Get UsePrepTimeInterpolation())
   Set CurrPrepTimeLeftSec(execute=@n7.else, CurrPrepTimeLeftSec=Get UpdatePrepTimeLeftSec())
   n8 = Get UpdatePrepTimeLeftSec()
   n9 = Get CurrTravelStatus() -> (CurrTravelStatus=None)
   FortShippingLog(
    execute=@n6.then_1
    WorldContextObject=Self-Reference()
    InString=Append!(A="[PlaylistTravel] [HUD] UpdatePrepTime: ", B="To String (Float)"(InDouble=@n8), C:String=" Sec, ", D:String=Enum to String(Enumerator=@n9)) -> (ReturnValue)
    bLogAsWarning=false
   )
  ]}
 UpdateProgressBar {nodes [
   UpdateProgressBar!() -> (then, Progress:double=~) {is=title K2Node_FunctionEntry}
   > n1 = GetDynamicMaterial(self:Border Object Reference=Get Border_Backing() -> (Border_Backing:Border Object Reference))
   > SetScalarParameterValue(self=@n1, ParameterName=Progress Amount, Value=Get Progress() -> (Progress:double))
  ]}
]
variables [
 CurrPrepTimeLeftSec {type=number}
 CurrTravelStatus {type=EFortPlaylistTravelStatus, default=None}
 LaunchDotUpdateTime {type=number, default=0.5}
 LaunchingText {type=string, default=Launching}
 LaunchingTextDotCount {type=integer}
 MatchmakingText {type=string, default=Matchmaking in progress...}
 MaxPrepTimeSec {type=integer, default=5}
 PreparingText {type=string, default=Preparing to travel...}
 PrepTimeFixInterpolateSpeed {type=number, default=5}
 PrepTimeTickSec {type=number, default=0.05}
 ProgressTime01 {type=number}
 Timer_Launching {type=TimerHandle}
 Timer_Preptime_Progress {type=TimerHandle}
 TotalPrepTimeLeftSec {type=number, default=1}
 UpdatePrepTimeLeftSec {type=number}
 UsePrepTimeInterpolation {type=boolean, default=true}
]
widgets {
 info {
  inheritedWidgetCount=1
  parentClass=&FortHUDElementWidget
  rootWidgetClass=/Script/UIFramework.UIFrameworkButtonWidget
  widgetCount=25
 }
 widgets [
  UIFrameworkButtonWidget_Toast {widgetClassPath=&UIFrameworkButtonWidget, bIsVariable=true, widget=UIFrameworkButtonWidget_Toast}
  Border_Backing {
   widgetClassPath=&Border
   bIsVariable=true
   parent=UIFrameworkButtonWidget_Toast
   widget=Border_Backing
   properties {
    visibility=SelfHitTestInvisible
    background {
     bIsDynamicallyLoaded=false
     drawAs=Box
     imageType=FullColor
     mirroring=NoMirror
     resourceName=None
     resourceObject=/FortPlaylistTravelUI/Art/MI_UIKit_Fort_Toast_Backplate_Inst
     tiling=NoTile
     imageSize=<32 32>
     margin=<ltrb 0.5 0.5 0.5 0.5>
     outlineSettings {
      bUseBrushTransparency=false
      roundingType=HalfHeightRadius
      width=0
      color {colorUseRule=UseColor_Specified, specifiedColor=<rgba 0 0 0 0>}
      cornerRadii=<0 0 0 0>
     }
     tintColor {colorUseRule=UseColor_Specified, specifiedColor=<rgba 1 1 1 1>}
     uVRegion {bIsValid=false, max=<0 0>, min=<0 0>}
    }
    padding=<ltrb 16 16 16 20>
   }
   slot {class=ButtonSlot}
  }
  VerticalBox_1 {
   widgetClassPath=&VerticalBox
   parent=Border_Backing
   widget=VerticalBox_1
   properties {bOverride_Cursor=true}
   slot {class=BorderSlot, properties {padding=<ltrb 16 16 16 20>}}
  }
  HorizontalBox_39 {widgetClassPath=&HorizontalBox, parent=VerticalBox_1, widget=HorizontalBox_39, slot {class=VerticalBoxSlot}}
  DestinationPreview {
   widgetClassPath=&FortLazyImage
   bIsVariable=true
   parent=HorizontalBox_39
   widget=DestinationPreview
   properties {
    materialTextureParamName=Image
    brush {
     bIsDynamicallyLoaded=false
     drawAs=Image
     imageType=FullColor
     mirroring=NoMirror
     resourceName=None
     resourceObject=/FortPlaylistTravelUI/Art/MI_UI_FortPlaylistTravel_GamePreviewImage_S
     tiling=NoTile
     imageSize=<80 80>
     margin=<ltrb 0.5 0.5 0.5 0.5>
     outlineSettings {
      bUseBrushTransparency=false
      roundingType=HalfHeightRadius
      width=0
      color {colorUseRule=UseColor_Specified, specifiedColor=<rgba 0 0 0 0>}
      cornerRadii=<10 10 10 10>
     }
     tintColor {colorUseRule=UseColor_Specified, specifiedColor=<rgba 1 1 1 1>}
     uVRegion {bIsValid=false, max=<0 0>, min=<0 0>}
    }
   }
   slot {class=HorizontalBoxSlot, properties {horizontalAlignment=HAlign_Left, verticalAlignment=VAlign_Center, padding=<ltrb 0 0 16 0>}}
  }
  SB_ClippingEnforcement_Download {
   widgetClassPath=&SizeBox
   bIsVariable=true
   parent=HorizontalBox_39
   widget=SB_ClippingEnforcement_Download
   properties {bOverride_MaxDesiredWidth=true, bOverride_MinDesiredWidth=true, maxDesiredWidth=380, minDesiredWidth=300, widthOverride=400}
   slot {class=HorizontalBoxSlot, properties {verticalAlignment=VAlign_Center}}
  }
  VerticalBox {widgetClassPath=&VerticalBox, parent=HorizontalBox_39, widget=VerticalBox, slot {class=HorizontalBoxSlot}}
  HorizontalBox {widgetClassPath=&HorizontalBox, parent=HorizontalBox_39, widget=HorizontalBox, slot {class=HorizontalBoxSlot}}
  WBP_UIKit_Throbber {
   widgetClassPath=/FortUILibrary/UIKit/Throbber/WBP_UIKit_Throbber
   bIsVariable=true
   parent=HorizontalBox_39
   widget=WBP_UIKit_Throbber
   properties {displayBackground=false}
   slot {class=HorizontalBoxSlot, properties {horizontalAlignment=HAlign_Center, verticalAlignment=VAlign_Center}}
  }
  Text_Downloading {
   widgetClassPath=&CommonRichTextBlock
   parent=HorizontalBox_39
   widget=Text_Downloading
   properties {
    bOverrideDefaultStyle=true
    clipping=ClipToBounds
    scrollStyle=&TextScrollStyle-Base_C
    text=Downloading
    textOverflowPolicy=Ellipsis
    defaultTextStyleOverride {
     overflowPolicy=Clip
     transformPolicy=None
     colorAndOpacity {colorUseRule=UseColor_Specified, specifiedColor=<rgba 1 1 1 1>}
     font {
      bForceMonospaced=false
      bMaterialIsStencil=false
      fontMaterial=null
      fontObject=&HeadingNow
      letterSpacing=0
      monospacedWidth=1
      size=18
      skewAmount=0
      typefaceFontName=75 Medium
      outlineSettings {
       bApplyOutlineToDropShadows=false
       bMiteredCorners=false
       bSeparateFillAlpha=false
       outlineMaterial=null
       outlineSize=2
       outlineColor=<rgba 0 0 0 0.3>
      }
     }
     highlightColor {colorUseRule=UseColor_Specified, specifiedColor=<rgba 0 0 0 1>}
     highlightShape {
      bIsDynamicallyLoaded=false
      drawAs=Image
      imageType=NoImage
      mirroring=NoMirror
      resourceName=None
      resourceObject=null
      tiling=NoTile
      imageSize=<32 32>
      margin=<ltrb 0 0 0 0>
      outlineSettings {
       bUseBrushTransparency=false
       roundingType=HalfHeightRadius
       width=0
       color {colorUseRule=UseColor_Specified, specifiedColor=<rgba 0 0 0 0>}
       cornerRadii=<0 0 0 0>
      }
      tintColor {colorUseRule=UseColor_Specified, specifiedColor=<rgba 1 1 1 1>}
      uVRegion {bIsValid=false, max=<0 0>, min=<0 0>}
     }
     selectedBackgroundColor {colorUseRule=UseColor_Foreground, specifiedColor=<rgba 1 0 1 1>}
     shadowColorAndOpacity=<rgba 0 0 0 1>
     shadowOffset=<0 0>
     strikeBrush {
      bIsDynamicallyLoaded=false
      drawAs=Image
      imageType=NoImage
      mirroring=NoMirror
      resourceName=None
      resourceObject=null
      tiling=NoTile
      imageSize=<32 32>
      margin=<ltrb 0 0 0 0>
      outlineSettings {
       bUseBrushTransparency=false
       roundingType=HalfHeightRadius
       width=0
       color {colorUseRule=UseColor_Specified, specifiedColor=<rgba 0 0 0 0>}
       cornerRadii=<0 0 0 0>
      }
      tintColor {colorUseRule=UseColor_Specified, specifiedColor=<rgba 1 1 1 1>}
      uVRegion {bIsValid=false, max=<0 0>, min=<0 0>}
     }
     underlineBrush {
      bIsDynamicallyLoaded=false
      drawAs=Image
      imageType=NoImage
      mirroring=NoMirror
      resourceName=None
      resourceObject=null
      tiling=NoTile
      imageSize=<32 32>
      margin=<ltrb 0 0 0 0>
      outlineSettings {
       bUseBrushTransparency=false
       roundingType=HalfHeightRadius
       width=0
       color {colorUseRule=UseColor_Specified, specifiedColor=<rgba 0 0 0 0>}
       cornerRadii=<0 0 0 0>
      }
      tintColor {colorUseRule=UseColor_Specified, specifiedColor=<rgba 1 1 1 1>}
      uVRegion {bIsValid=false, max=<0 0>, min=<0 0>}
     }
    }
    margin=<ltrb 4 0 4 2>
   }
   slot {class=HorizontalBoxSlot, properties {horizontalAlignment=HAlign_Center, verticalAlignment=VAlign_Center}}
  }
  SB_ClippingEnforcement_TravelInfo {
   widgetClassPath=&SizeBox
   bIsVariable=true
   parent=HorizontalBox_39
   widget=SB_ClippingEnforcement_TravelInfo
   properties {bOverride_MaxDesiredWidth=true, bOverride_MinDesiredWidth=true, maxDesiredWidth=380, minDesiredWidth=300, widthOverride=400}
   slot {class=HorizontalBoxSlot, properties {verticalAlignment=VAlign_Center}}
  }
  VerticalBox_0 {widgetClassPath=&VerticalBox, parent=SB_ClippingEnforcement_TravelInfo, widget=VerticalBox_0, slot {class=SizeBoxSlot}}
  Text_TravelDestination {
   widgetClassPath=&CommonRichTextBlock
   bIsVariable=true
   parent=VerticalBox_0
   widget=Text_TravelDestination
   properties {
    bOverrideDefaultStyle=true
    clipping=ClipToBounds
    scrollStyle=&TextScrollStyle-Base_C
    text="[PH] Destination Testing Slipping"
    textOverflowPolicy=Ellipsis
    defaultTextStyleOverride {
     overflowPolicy=Clip
     transformPolicy=None
     colorAndOpacity {colorUseRule=UseColor_Specified, specifiedColor=<rgba 1 1 1 1>}
     font {
      bForceMonospaced=false
      bMaterialIsStencil=false
      fontMaterial=null
      fontObject=&HeadingNow
      letterSpacing=0
      monospacedWidth=1
      size=18
      skewAmount=0
      typefaceFontName=75 Medium
      outlineSettings {
       bApplyOutlineToDropShadows=false
       bMiteredCorners=false
       bSeparateFillAlpha=false
       outlineMaterial=null
       outlineSize=2
       outlineColor=<rgba 0 0 0 0.3>
      }
     }
     highlightColor {colorUseRule=UseColor_Specified, specifiedColor=<rgba 0 0 0 1>}
     highlightShape {
      bIsDynamicallyLoaded=false
      drawAs=Image
      imageType=NoImage
      mirroring=NoMirror
      resourceName=None
      resourceObject=null
      tiling=NoTile
      imageSize=<32 32>
      margin=<ltrb 0 0 0 0>
      outlineSettings {
       bUseBrushTransparency=false
       roundingType=HalfHeightRadius
       width=0
       color {colorUseRule=UseColor_Specified, specifiedColor=<rgba 0 0 0 0>}
       cornerRadii=<0 0 0 0>
      }
      tintColor {colorUseRule=UseColor_Specified, specifiedColor=<rgba 1 1 1 1>}
      uVRegion {bIsValid=false, max=<0 0>, min=<0 0>}
     }
     selectedBackgroundColor {colorUseRule=UseColor_Foreground, specifiedColor=<rgba 1 0 1 1>}
     shadowColorAndOpacity=<rgba 0 0 0 1>
     shadowOffset=<0 0>
     strikeBrush {
      bIsDynamicallyLoaded=false
      drawAs=Image
      imageType=NoImage
      mirroring=NoMirror
      resourceName=None
      resourceObject=null
      tiling=NoTile
      imageSize=<32 32>
      margin=<ltrb 0 0 0 0>
      outlineSettings {
       bUseBrushTransparency=false
       roundingType=HalfHeightRadius
       width=0
       color {colorUseRule=UseColor_Specified, specifiedColor=<rgba 0 0 0 0>}
       cornerRadii=<0 0 0 0>
      }
      tintColor {colorUseRule=UseColor_Specified, specifiedColor=<rgba 1 1 1 1>}
      uVRegion {bIsValid=false, max=<0 0>, min=<0 0>}
     }
     underlineBrush {
      bIsDynamicallyLoaded=false
      drawAs=Image
      imageType=NoImage
      mirroring=NoMirror
      resourceName=None
      resourceObject=null
      tiling=NoTile
      imageSize=<32 32>
      margin=<ltrb 0 0 0 0>
      outlineSettings {
       bUseBrushTransparency=false
       roundingType=HalfHeightRadius
       width=0
       color {colorUseRule=UseColor_Specified, specifiedColor=<rgba 0 0 0 0>}
       cornerRadii=<0 0 0 0>
      }
      tintColor {colorUseRule=UseColor_Specified, specifiedColor=<rgba 1 1 1 1>}
      uVRegion {bIsValid=false, max=<0 0>, min=<0 0>}
     }
    }
    margin=<ltrb 4 0 4 2>
   }
   slot {class=VerticalBoxSlot}
  }
  HorizontalBox_1 {widgetClassPath=&HorizontalBox, parent=VerticalBox_0, widget=HorizontalBox_1, slot {class=VerticalBoxSlot}}
  Text_Timer {
   widgetClassPath=&CommonRichTextBlock
   bIsVariable=true
   parent=HorizontalBox_1
   widget=Text_Timer
   properties {
    bOverrideDefaultStyle=true
    scrollStyle=&TextScrollStyle-Base_C
    text="{0:00}"
    textOverflowPolicy=Ellipsis
    decoratorClasses [null]
    defaultTextStyleOverride {
     overflowPolicy=Clip
     transformPolicy=None
     colorAndOpacity {colorUseRule=UseColor_Specified, specifiedColor=<rgba 0 1 0.9647 1>}
     font {
      bForceMonospaced=false
      bMaterialIsStencil=false
      fontMaterial=null
      fontObject=&HeadingNow
      letterSpacing=0
      monospacedWidth=1
      size=15.75
      skewAmount=0
      typefaceFontName=74 Regular
      outlineSettings {
       bApplyOutlineToDropShadows=false
       bMiteredCorners=false
       bSeparateFillAlpha=false
       outlineMaterial=null
       outlineSize=2
       outlineColor=<rgba 0 0 0 0.3>
      }
     }
     highlightColor {colorUseRule=UseColor_Specified, specifiedColor=<rgba 0 0 0 1>}
     highlightShape {
      bIsDynamicallyLoaded=false
      drawAs=Image
      imageType=NoImage
      mirroring=NoMirror
      resourceName=None
      resourceObject=null
      tiling=NoTile
      imageSize=<32 32>
      margin=<ltrb 0 0 0 0>
      outlineSettings {
       bUseBrushTransparency=false
       roundingType=HalfHeightRadius
       width=0
       color {colorUseRule=UseColor_Specified, specifiedColor=<rgba 0 0 0 0>}
       cornerRadii=<0 0 0 0>
      }
      tintColor {colorUseRule=UseColor_Specified, specifiedColor=<rgba 1 1 1 1>}
      uVRegion {bIsValid=false, max=<0 0>, min=<0 0>}
     }
     selectedBackgroundColor {colorUseRule=UseColor_Foreground, specifiedColor=<rgba 1 0 1 1>}
     shadowColorAndOpacity=<rgba 0 0 0 1>
     shadowOffset=<0 0>
     strikeBrush {
      bIsDynamicallyLoaded=false
      drawAs=Image
      imageType=NoImage
      mirroring=NoMirror
      resourceName=None
      resourceObject=null
      tiling=NoTile
      imageSize=<32 32>
      margin=<ltrb 0 0 0 0>
      outlineSettings {
       bUseBrushTransparency=false
       roundingType=HalfHeightRadius
       width=0
       color {colorUseRule=UseColor_Specified, specifiedColor=<rgba 0 0 0 0>}
       cornerRadii=<0 0 0 0>
      }
      tintColor {colorUseRule=UseColor_Specified, specifiedColor=<rgba 1 1 1 1>}
      uVRegion {bIsValid=false, max=<0 0>, min=<0 0>}
     }
     underlineBrush {
      bIsDynamicallyLoaded=false
      drawAs=Image
      imageType=NoImage
      mirroring=NoMirror
      resourceName=None
      resourceObject=null
      tiling=NoTile
      imageSize=<32 32>
      margin=<ltrb 0 0 0 0>
      outlineSettings {
       bUseBrushTransparency=false
       roundingType=HalfHeightRadius
       width=0
       color {colorUseRule=UseColor_Specified, specifiedColor=<rgba 0 0 0 0>}
       cornerRadii=<0 0 0 0>
      }
      tintColor {colorUseRule=UseColor_Specified, specifiedColor=<rgba 1 1 1 1>}
      uVRegion {bIsValid=false, max=<0 0>, min=<0 0>}
     }
    }
    margin=<ltrb 4 0 4 2>
   }
   slot {class=HorizontalBoxSlot, properties {padding=<ltrb 0 0 2 0>}}
  }
  Text_TravelState {
   widgetClassPath=&CommonRichTextBlock
   bIsVariable=true
   parent=HorizontalBox_1
   widget=Text_TravelState
   properties {
    bOverrideDefaultStyle=true
    clipping=ClipToBounds
    scrollStyle=&TextScrollStyle-Base_C
    text="[PH] Prepare for travel"
    textOverflowPolicy=Ellipsis
    decoratorClasses [null]
    defaultTextStyleOverride {
     overflowPolicy=Clip
     transformPolicy=None
     colorAndOpacity {colorUseRule=UseColor_Specified, specifiedColor=<rgba 1 1 1 0.7>}
     font {
      bForceMonospaced=false
      bMaterialIsStencil=false
      fontMaterial=null
      fontObject=&HeadingNow
      letterSpacing=0
      monospacedWidth=1
      size=15.75
      skewAmount=0
      typefaceFontName=74 Regular
      outlineSettings {
       bApplyOutlineToDropShadows=false
       bMiteredCorners=false
       bSeparateFillAlpha=false
       outlineMaterial=null
       outlineSize=2
       outlineColor=<rgba 0 0 0 0.3>
      }
     }
     highlightColor {colorUseRule=UseColor_Specified, specifiedColor=<rgba 0 0 0 1>}
     highlightShape {
      bIsDynamicallyLoaded=false
      drawAs=Image
      imageType=NoImage
      mirroring=NoMirror
      resourceName=None
      resourceObject=null
      tiling=NoTile
      imageSize=<32 32>
      margin=<ltrb 0 0 0 0>
      outlineSettings {
       bUseBrushTransparency=false
       roundingType=HalfHeightRadius
       width=0
       color {colorUseRule=UseColor_Specified, specifiedColor=<rgba 0 0 0 0>}
       cornerRadii=<0 0 0 0>
      }
      tintColor {colorUseRule=UseColor_Specified, specifiedColor=<rgba 1 1 1 1>}
      uVRegion {bIsValid=false, max=<0 0>, min=<0 0>}
     }
     selectedBackgroundColor {colorUseRule=UseColor_Foreground, specifiedColor=<rgba 1 0 1 1>}
     shadowColorAndOpacity=<rgba 0 0 0 1>
     shadowOffset=<0 0>
     strikeBrush {
      bIsDynamicallyLoaded=false
      drawAs=Image
      imageType=NoImage
      mirroring=NoMirror
      resourceName=None
      resourceObject=null
      tiling=NoTile
      imageSize=<32 32>
      margin=<ltrb 0 0 0 0>
      outlineSettings {
       bUseBrushTransparency=false
       roundingType=HalfHeightRadius
       width=0
       color {colorUseRule=UseColor_Specified, specifiedColor=<rgba 0 0 0 0>}
       cornerRadii=<0 0 0 0>
      }
      tintColor {colorUseRule=UseColor_Specified, specifiedColor=<rgba 1 1 1 1>}
      uVRegion {bIsValid=false, max=<0 0>, min=<0 0>}
     }
     underlineBrush {
      bIsDynamicallyLoaded=false
      drawAs=Image
      imageType=NoImage
      mirroring=NoMirror
      resourceName=None
      resourceObject=null
      tiling=NoTile
      imageSize=<32 32>
      margin=<ltrb 0 0 0 0>
      outlineSettings {
       bUseBrushTransparency=false
       roundingType=HalfHeightRadius
       width=0
       color {colorUseRule=UseColor_Specified, specifiedColor=<rgba 0 0 0 0>}
       cornerRadii=<0 0 0 0>
      }
      tintColor {colorUseRule=UseColor_Specified, specifiedColor=<rgba 1 1 1 1>}
      uVRegion {bIsValid=false, max=<0 0>, min=<0 0>}
     }
    }
    margin=<ltrb 4 0 4 2>
   }
   slot {class=HorizontalBoxSlot}
  }
  Text_TravelState_Dot {
   widgetClassPath=&CommonRichTextBlock
   bIsVariable=true
   parent=HorizontalBox_1
   widget=Text_TravelState_Dot
   properties {
    bOverrideDefaultStyle=true
    clipping=ClipToBounds
    scrollStyle=&TextScrollStyle-Base_C
    textOverflowPolicy=Ellipsis
    visibility=Collapsed
    decoratorClasses [null]
    defaultTextStyleOverride {
     overflowPolicy=Clip
     transformPolicy=None
     colorAndOpacity {colorUseRule=UseColor_Specified, specifiedColor=<rgba 1 1 1 0.7>}
     font {
      bForceMonospaced=false
      bMaterialIsStencil=false
      fontMaterial=null
      fontObject=&HeadingNow
      letterSpacing=0
      monospacedWidth=1
      size=15.75
      skewAmount=0
      typefaceFontName=74 Regular
      outlineSettings {
       bApplyOutlineToDropShadows=false
       bMiteredCorners=false
       bSeparateFillAlpha=false
       outlineMaterial=null
       outlineSize=2
       outlineColor=<rgba 0 0 0 0.3>
      }
     }
     highlightColor {colorUseRule=UseColor_Specified, specifiedColor=<rgba 0 0 0 1>}
     highlightShape {
      bIsDynamicallyLoaded=false
      drawAs=Image
      imageType=NoImage
      mirroring=NoMirror
      resourceName=None
      resourceObject=null
      tiling=NoTile
      imageSize=<32 32>
      margin=<ltrb 0 0 0 0>
      outlineSettings {
       bUseBrushTransparency=false
       roundingType=HalfHeightRadius
       width=0
       color {colorUseRule=UseColor_Specified, specifiedColor=<rgba 0 0 0 0>}
       cornerRadii=<0 0 0 0>
      }
      tintColor {colorUseRule=UseColor_Specified, specifiedColor=<rgba 1 1 1 1>}
      uVRegion {bIsValid=false, max=<0 0>, min=<0 0>}
     }
     selectedBackgroundColor {colorUseRule=UseColor_Foreground, specifiedColor=<rgba 1 0 1 1>}
     shadowColorAndOpacity=<rgba 0 0 0 1>
     shadowOffset=<0 0>
     strikeBrush {
      bIsDynamicallyLoaded=false
      drawAs=Image
      imageType=NoImage
      mirroring=NoMirror
      resourceName=None
      resourceObject=null
      tiling=NoTile
      imageSize=<32 32>
      margin=<ltrb 0 0 0 0>
      outlineSettings {
       bUseBrushTransparency=false
       roundingType=HalfHeightRadius
       width=0
       color {colorUseRule=UseColor_Specified, specifiedColor=<rgba 0 0 0 0>}
       cornerRadii=<0 0 0 0>
      }
      tintColor {colorUseRule=UseColor_Specified, specifiedColor=<rgba 1 1 1 1>}
      uVRegion {bIsValid=false, max=<0 0>, min=<0 0>}
     }
     underlineBrush {
      bIsDynamicallyLoaded=false
      drawAs=Image
      imageType=NoImage
      mirroring=NoMirror
      resourceName=None
      resourceObject=null
      tiling=NoTile
      imageSize=<32 32>
      margin=<ltrb 0 0 0 0>
      outlineSettings {
       bUseBrushTransparency=false
       roundingType=HalfHeightRadius
       width=0
       color {colorUseRule=UseColor_Specified, specifiedColor=<rgba 0 0 0 0>}
       cornerRadii=<0 0 0 0>
      }
      tintColor {colorUseRule=UseColor_Specified, specifiedColor=<rgba 1 1 1 1>}
      uVRegion {bIsValid=false, max=<0 0>, min=<0 0>}
     }
    }
    margin=<ltrb 4 0 4 2>
   }
   slot {class=HorizontalBoxSlot, properties {padding=<ltrb -10 0 0 0>}}
  }
  SquadOptOutHorizontalBox {
   widgetClassPath=&HorizontalBox
   bIsVariable=true
   parent=VerticalBox_1
   widget=SquadOptOutHorizontalBox
   slot {class=VerticalBoxSlot}
  }
  CommonVisibilityWidgetLegacy_0 {
   widgetClassPath=&CommonVisibilityWidgetLegacy
   parent=SquadOptOutHorizontalBox
   widget=CommonVisibilityWidgetLegacy_0
   properties {bShowForTouch=false, padding=<ltrb 0 0 0 0>}
   slot {class=HorizontalBoxSlot, properties {horizontalAlignment=HAlign_Left, verticalAlignment=VAlign_Center, padding=<ltrb 23 0 0 0>}}
  }
  Overlay_0 {
   widgetClassPath=&Overlay
   parent=CommonVisibilityWidgetLegacy_0
   widget=Overlay_0
   slot {class=BorderSlot, properties {padding=<ltrb 0 0 0 0>}}
  }
  KeybindContainer {
   widgetClassPath=&ScaleBox
   bIsVariable=true
   parent=Overlay_0
   widget=KeybindContainer
   properties {stretch=UserSpecified, userSpecifiedScale=0.75}
   slot {class=OverlaySlot, properties {horizontalAlignment=HAlign_Center, verticalAlignment=VAlign_Center}}
  }
  KeybindWidget_OptOut {
   widgetClassPath=/FortUILibrary/Input/KeybindWidget
   bIsVariable=true
   parent=KeybindContainer
   widget=KeybindWidget_OptOut
   properties {
    bUseKeyboardTextureForKeys=true
    boundAction=Pilgrim_ToggleReady
    boundInputAction=/FortPlaylistTravel/Input/IA_OptOutOfSquadTravel
    forcedHoldKeybindStatus=ForcedHold
    visibility=Visible
   }
   slot {class=ScaleBoxSlot}
  }
  Text_SquadMemberOptOut {
   widgetClassPath=&CommonRichTextBlock
   bIsVariable=true
   parent=SquadOptOutHorizontalBox
   widget=Text_SquadMemberOptOut
   properties {
    bOverrideDefaultStyle=true
    clipping=ClipToBounds
    scrollStyle=&TextScrollStyle-Base_C
    text=Hold Button to Opt Out of Squad Travel
    textOverflowPolicy=Ellipsis
    decoratorClasses [null]
    defaultTextStyleOverride {
     overflowPolicy=Clip
     transformPolicy=None
     colorAndOpacity {colorUseRule=UseColor_Specified, specifiedColor=<rgba 1 1 1 0.7>}
     font {
      bForceMonospaced=false
      bMaterialIsStencil=false
      fontMaterial=null
      fontObject=&HeadingNow
      letterSpacing=0
      monospacedWidth=1
      size=15.75
      skewAmount=0
      typefaceFontName=74 Regular
      outlineSettings {
       bApplyOutlineToDropShadows=false
       bMiteredCorners=false
       bSeparateFillAlpha=false
       outlineMaterial=null
       outlineSize=2
       outlineColor=<rgba 0 0 0 0.3>
      }
     }
     highlightColor {colorUseRule=UseColor_Specified, specifiedColor=<rgba 0 0 0 1>}
     highlightShape {
      bIsDynamicallyLoaded=false
      drawAs=Image
      imageType=NoImage
      mirroring=NoMirror
      resourceName=None
      resourceObject=null
      tiling=NoTile
      imageSize=<32 32>
      margin=<ltrb 0 0 0 0>
      outlineSettings {
       bUseBrushTransparency=false
       roundingType=HalfHeightRadius
       width=0
       color {colorUseRule=UseColor_Specified, specifiedColor=<rgba 0 0 0 0>}
       cornerRadii=<0 0 0 0>
      }
      tintColor {colorUseRule=UseColor_Specified, specifiedColor=<rgba 1 1 1 1>}
      uVRegion {bIsValid=false, max=<0 0>, min=<0 0>}
     }
     selectedBackgroundColor {colorUseRule=UseColor_Foreground, specifiedColor=<rgba 1 0 1 1>}
     shadowColorAndOpacity=<rgba 0 0 0 1>
     shadowOffset=<0 0>
     strikeBrush {
      bIsDynamicallyLoaded=false
      drawAs=Image
      imageType=NoImage
      mirroring=NoMirror
      resourceName=None
      resourceObject=null
      tiling=NoTile
      imageSize=<32 32>
      margin=<ltrb 0 0 0 0>
      outlineSettings {
       bUseBrushTransparency=false
       roundingType=HalfHeightRadius
       width=0
       color {colorUseRule=UseColor_Specified, specifiedColor=<rgba 0 0 0 0>}
       cornerRadii=<0 0 0 0>
      }
      tintColor {colorUseRule=UseColor_Specified, specifiedColor=<rgba 1 1 1 1>}
      uVRegion {bIsValid=false, max=<0 0>, min=<0 0>}
     }
     underlineBrush {
      bIsDynamicallyLoaded=false
      drawAs=Image
      imageType=NoImage
      mirroring=NoMirror
      resourceName=None
      resourceObject=null
      tiling=NoTile
      imageSize=<32 32>
      margin=<ltrb 0 0 0 0>
      outlineSettings {
       bUseBrushTransparency=false
       roundingType=HalfHeightRadius
       width=0
       color {colorUseRule=UseColor_Specified, specifiedColor=<rgba 0 0 0 0>}
       cornerRadii=<0 0 0 0>
      }
      tintColor {colorUseRule=UseColor_Specified, specifiedColor=<rgba 1 1 1 1>}
      uVRegion {bIsValid=false, max=<0 0>, min=<0 0>}
     }
    }
    margin=<ltrb 4 0 4 2>
   }
   slot {class=HorizontalBoxSlot}
  }
  Text_SquadInitiatorOptOut {
   widgetClassPath=&CommonRichTextBlock
   bIsVariable=true
   parent=SquadOptOutHorizontalBox
   widget=Text_SquadInitiatorOptOut
   properties {
    bOverrideDefaultStyle=true
    clipping=ClipToBounds
    scrollStyle=&TextScrollStyle-Base_C
    text=Hold Button to Cancel Squad Travel
    textOverflowPolicy=Ellipsis
    decoratorClasses [null]
    defaultTextStyleOverride {
     overflowPolicy=Clip
     transformPolicy=None
     colorAndOpacity {colorUseRule=UseColor_Specified, specifiedColor=<rgba 1 1 1 0.7>}
     font {
      bForceMonospaced=false
      bMaterialIsStencil=false
      fontMaterial=null
      fontObject=&HeadingNow
      letterSpacing=0
      monospacedWidth=1
      size=15.75
      skewAmount=0
      typefaceFontName=74 Regular
      outlineSettings {
       bApplyOutlineToDropShadows=false
       bMiteredCorners=false
       bSeparateFillAlpha=false
       outlineMaterial=null
       outlineSize=2
       outlineColor=<rgba 0 0 0 0.3>
      }
     }
     highlightColor {colorUseRule=UseColor_Specified, specifiedColor=<rgba 0 0 0 1>}
     highlightShape {
      bIsDynamicallyLoaded=false
      drawAs=Image
      imageType=NoImage
      mirroring=NoMirror
      resourceName=None
      resourceObject=null
      tiling=NoTile
      imageSize=<32 32>
      margin=<ltrb 0 0 0 0>
      outlineSettings {
       bUseBrushTransparency=false
       roundingType=HalfHeightRadius
       width=0
       color {colorUseRule=UseColor_Specified, specifiedColor=<rgba 0 0 0 0>}
       cornerRadii=<0 0 0 0>
      }
      tintColor {colorUseRule=UseColor_Specified, specifiedColor=<rgba 1 1 1 1>}
      uVRegion {bIsValid=false, max=<0 0>, min=<0 0>}
     }
     selectedBackgroundColor {colorUseRule=UseColor_Foreground, specifiedColor=<rgba 1 0 1 1>}
     shadowColorAndOpacity=<rgba 0 0 0 1>
     shadowOffset=<0 0>
     strikeBrush {
      bIsDynamicallyLoaded=false
      drawAs=Image
      imageType=NoImage
      mirroring=NoMirror
      resourceName=None
      resourceObject=null
      tiling=NoTile
      imageSize=<32 32>
      margin=<ltrb 0 0 0 0>
      outlineSettings {
       bUseBrushTransparency=false
       roundingType=HalfHeightRadius
       width=0
       color {colorUseRule=UseColor_Specified, specifiedColor=<rgba 0 0 0 0>}
       cornerRadii=<0 0 0 0>
      }
      tintColor {colorUseRule=UseColor_Specified, specifiedColor=<rgba 1 1 1 1>}
      uVRegion {bIsValid=false, max=<0 0>, min=<0 0>}
     }
     underlineBrush {
      bIsDynamicallyLoaded=false
      drawAs=Image
      imageType=NoImage
      mirroring=NoMirror
      resourceName=None
      resourceObject=null
      tiling=NoTile
      imageSize=<32 32>
      margin=<ltrb 0 0 0 0>
      outlineSettings {
       bUseBrushTransparency=false
       roundingType=HalfHeightRadius
       width=0
       color {colorUseRule=UseColor_Specified, specifiedColor=<rgba 0 0 0 0>}
       cornerRadii=<0 0 0 0>
      }
      tintColor {colorUseRule=UseColor_Specified, specifiedColor=<rgba 1 1 1 1>}
      uVRegion {bIsValid=false, max=<0 0>, min=<0 0>}
     }
    }
    margin=<ltrb 4 0 4 2>
   }
   slot {class=HorizontalBoxSlot}
  }
  TutorialHighlight {widgetClassPath=&TutorialHighlight}
 ]
}
